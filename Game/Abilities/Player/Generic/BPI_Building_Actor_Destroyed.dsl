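; Blueprint interface for player generic abilities that declares a single event for when a building actor is destroyed.
; It exposes Event_Building_Actor_Destroyed, which takes no inputs and returns nothing.
; The file holds only the declaration; implementers supply the response and callers send it as a notification.
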(fn Event_Building_Actor_Destroyed)
